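# Unlit, two-sided translucent Niagara mesh-particle material for nitro speed fire.
# Emissive blends two colors along V, tinted by particle color, with optional HDR remapping.
# Opacity multiplies three scales of panning, per-particle time-offset directional noise, shaped by gradient erosion, edge and sphere masks, then view-angle and depth fades.
# Dynamic parameters drive fire intensity, tiling and a clamped, UV-weighted world position offset derived from particle position, bending the mesh.
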Material /Game/Effects/Niagara/MasterMaterials/Fire/NitroFire/M_Nitro_SpeedFire
graph {
 properties {
  bUsedWithNiagaraMeshParticles=true
  blendMode=BLEND_Translucent
  pixelDepthOffsetMode=PDOM_Legacy
  shadingModel=MSM_Unlit
  twoSided=true
  usedShadingModels=Unlit
 }
 nodes [
  n1 = DynamicParameter() {desc=Parameters set by Niagara, parameterIndex=1, paramNames [X,Y,Z,WPOIntensity]}
  n2 = Subtract(
   A=MaterialFunctionCall(X=@n1.X, Y=@n1.X, Z=@n1.X) {materialFunction=&MakeFloat3}.Result
   B=ParticlePositionWS()
  )
  n3 = ScalarParameter() {defaultValue=-100, parameterName=Clamp_min}
  n4 = ScalarParameter() {defaultValue=100, parameterName=Clamp_max}
  n5 = TextureCoordinate()
  n6 = Power(Base=ComponentMask(None=OneMinus(None=@n5)) {g=true}, Exp=ScalarParameter() {defaultValue=3, parameterName=WPOBendPower})
  LinearInterpolate(
   A=Constant()
   B=LinearInterpolate(B=Clamp(None=@n2, Min=@n3, Max=@n4) {maxDefault=1000, minDefault=-1000}, Alpha=Multiply(A=@n1.WPOIntensity, B=@n6))
   Alpha=ScalarParameter() {defaultValue=1, parameterName=Use interpolate WPO}
  )
  n7 = Power(
   Base=MaterialFunctionCall() {materialFunction=&LinearGradient}.VGradient
   Exp=Constant() {r=0.4}
  )
  n8 = Constant() {r=1.05}
  n9 = LinearInterpolate(
   A=VectorParameter() {parameterName=ColorA, defaultValue=<rgba 1 0.837 0.1 1>}.RGB
   B=Multiply(A=VectorParameter() {parameterName=ColorB, defaultValue=<rgba 1 0.0632 0.016 1>}.RGB, B=Constant() {r=0.8})
   Alpha=Saturate(None=Multiply(A=@n7, B=@n8) {constB=1.5})
  )
  n10 = ParticleColor()
  n11 = Saturate(None=@n9)
  MaterialFunctionCall(Emissive=LinearInterpolate(
    A=Multiply(A=@n9, B=@n10.RGB)
    B=MaterialFunctionCall(LDR Input=@n11, HDR Tint=@n10.RGB) {materialFunction=&DeriveHDRfromLDR}.RemappedRange
    Alpha=ScalarParameter() {parameterName=DeriveHDRfromLDR}
   )) {materialFunction=&MF_FX_EmissiveAdjustments}
  n12 = MaterialFunctionCall() {materialFunction=&LinearGradient}
  n13 = Power(Base=@n12.VGradient) {constExponent=0.5}
  n14 = AppendVector(
   A=ScalarParameter() {defaultValue=2, group=Tile, parameterName=TireU}
   B=ScalarParameter() {defaultValue=0.8, group=Tile, parameterName=TireV}
  )
  n15 = Divide(A=@n14) {constB=1.5}
  n16 = DynamicParameter() {paramNames [FiresIntensity, TileUandV, Param3, Param4]}
  n17 = Time()
  n18 = Multiply(A=ParticleRandom()) {constB=10}
  n19 = Panner(
   Coordinate=Multiply(A=AppendVector(A=@n12.UGradient, B=@n13), B=LinearInterpolate(A=@n15, B=@n14, Alpha=@n16.TileUandV))
   Time=Multiply(A=Add(A=@n17, B=@n18), B=ScalarParameter() {defaultValue=0.8, group=UV_PanningSpeed, parameterName=SpeedXandY}) {constB=0.5}
   Speed=AppendVector(A=ScalarParameter() {defaultValue=-1, group=UV_PanningSpeed, parameterName=SpeedX}, B=Constant() {r=2})
  ) {speedX=-1, speedY=2}
  n20 = Divide(A=@n19)
  n21 = Multiply(
   A=Multiply(
    A=TextureSample(UVs=@n19) {samplerType=SAMPLERTYPE_LinearGrayscale, texture=/Game/Effects/Fort_Effects/Emotes/UltraLight/Texture/T_DirectionalNoise_Greyscale}.R
    B=TextureSample(UVs=@n20) {samplerType=SAMPLERTYPE_LinearGrayscale, texture=&T_DirectionalNoise_Greyscale}.R
   )
   B=TextureSample(UVs=Divide(A=@n20) {constB=1.5}) {samplerType=SAMPLERTYPE_LinearGrayscale, texture=&T_DirectionalNoise_Greyscale}.R
  )
  n22 = MaterialFunctionCall() {materialFunction=&LinearGradient}
  n23 = Constant() {r=2}
  n24 = Multiply(A=MaterialFunctionCall() {materialFunction=&LinearGradient}.VGradient)
  n25 = DynamicParameter() {paramNames [FiresIntensity, TileUandV, Param3, Param4]}
  n26 = OneMinus(None=Multiply(A=ScalarParameter() {parameterName=Eroder}, B=LinearInterpolate(Alpha=@n25.FiresIntensity)))
  n27 = Subtract(
   A=SphereMask(A=Power(Base=@n22.VGradient, Exp=@n23), B=Constant() {r=0.5}) {attenuationRadius=0.5, hardnessPercent=0}
   B=Saturate(None=Subtract(A=@n24, B=@n26) {constB=0.5})
  )
  n28 = ScalarParameter() {defaultValue=10, parameterName=FireInentsity}
  n29 = DynamicParameter() {paramNames [FiresIntensity, TileUandV, Param3, Param4]}
  n30 = MaterialFunctionCall!(UVs, Mult, Exp=Constant() {r=2}) {materialFunction=/Game/Effects/Fort_Effects/MaterialFunctions/MF_SquareEdgeMask}
  n31 = Multiply(A=Multiply(A=Multiply(A=@n21, B=@n27), B=LinearInterpolate(B=@n28, Alpha=@n29.FiresIntensity)), B=Multiply(A=@n30.Vertmask, B=@n30.Vertmask))
  n32 = Multiply(
   A=MaterialFunctionCall!(
    Hardness=ScalarParameter() {parameterName=Hardness}
    Center=ScalarParameter() {defaultValue=0.5, parameterName=Center}
    Radius=ScalarParameter() {defaultValue=0.5, parameterName=Radius}
   ) {materialFunction=/Game/Packages/Fortress_SharedMaterials/Base_Material_Functions/FX_SphereMask2D}.Result
   B=ScalarParameter() {defaultValue=0.15, parameterName=SphereMask2D_Adjustment}
  )
  n33 = MaterialFunctionCall("FadeSensitivity (optional)"=ScalarParameter() {defaultValue=4, parameterName=FadeBasedOnViewAngle}) {materialFunction=&FadeBasedOnViewAngle}
  n34 = Constant() {r=2}
  n35 = Multiply(A=Saturate(None=Subtract(A=@n31, B=@n32)), B=Saturate(None=Multiply(A=@n33.Result, B=@n34)))
  Multiply(A=@n10.A, B=Multiply(A=DepthFade(Opacity=@n35), B=ScalarParameter() {defaultValue=20, parameterName=OverallOpacity}))
 ]
}
